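# Travel sound cue for the Cane event, layering movement loops, explosions and a pass-by under a high-priority events sound class.
# A random distant loop, a close loop and two delayed, looping random explosion layers each get distance low-pass and custom falloff.
# The distant curve rises from silence at the source before fading out, complementing the close layer.
# A louder, non-spatialized, short-range pass-by one-shot plays alongside them.
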SoundCue /Figment_S07/Athena/Sounds/Events/Cane/Cane_Travel_Cue
root=Mixer_1
nodes [
 {
  id=Attenuation_0
  children [Mixer_2]
  properties {
   bOverrideAttenuation=true
   attenuationOverrides {
    bAttenuateWithLPF=true
    distanceAlgorithm=Custom
    falloffDistance=4000
    lPFFrequencyAtMax=1000
    lPFRadiusMax=3000
    lPFRadiusMin=1000
    occlusionInterpolationTime=1
    occlusionLowPassFilterFrequency=1000
    occlusionTraceChannel=ECC_GameTraceChannel11
    stereoSpread=5000
    attenuationShapeExtents {x=0}
    customAttenuationCurve {editorCurveData {keys [
       {interpMode=RCIM_Cubic, leaveTangent=-2.893, tangentMode=RCTM_Break, time=0, value=1}
       {interpMode=RCIM_Cubic, tangentMode=RCTM_Break, time=1, value=0}
      ]}}
   }
  }
 }
 {
  id=Attenuation_1
  children [WavePlayer_2]
  properties {
   bOverrideAttenuation=true
   attenuationOverrides {
    bAttenuateWithLPF=true
    distanceAlgorithm=Custom
    falloffDistance=15000
    lPFFrequencyAtMax=3000
    lPFRadiusMax=8000
    lPFRadiusMin=2000
    occlusionInterpolationTime=1
    occlusionLowPassFilterFrequency=1000
    occlusionTraceChannel=ECC_GameTraceChannel11
    stereoSpread=5000
    attenuationShapeExtents {x=0}
    customAttenuationCurve {editorCurveData {keys [{arriveTangent=-2.982, interpMode=RCIM_Cubic, leaveTangent=-2.982, tangentMode=RCTM_User, time=0, value=1}, <at 1 0 RCIM_Cubic>]}}
   }
  }
 }
 {
  id=Attenuation_2
  children [SoundClass_1]
  properties {
   attenuationSettings=/Game/Athena/Sounds/Events/Attenuations/Events_Large_Close_Att
   bOverrideAttenuation=true
   attenuationOverrides {bSpatialize=false, falloffDistance=4000, stereoSpread=3000, attenuationShapeExtents {x=0}}
  }
 }
 {
  id=Attenuation_3
  children [Random_2]
  properties {
   bOverrideAttenuation=true
   attenuationOverrides {
    bAttenuateWithLPF=true
    distanceAlgorithm=Custom
    falloffDistance=17500
    lPFFrequencyAtMax=1000
    lPFRadiusMax=7500
    lPFRadiusMin=5000
    occlusionInterpolationTime=1
    occlusionLowPassFilterFrequency=1000
    occlusionTraceChannel=ECC_GameTraceChannel11
    stereoSpread=5000
    attenuationShapeExtents {x=0}
    customAttenuationCurve {editorCurveData {keys [<at 0 0>, {arriveTangent=-2.982, interpMode=RCIM_Cubic, leaveTangent=-2.982, tangentMode=RCTM_User, time=0.15, value=1}, <at 1 0 RCIM_Cubic>]}}
   }
  }
 }
 {id=Delay_0, children [Random_0], properties {delayMax=2, delayMin=0.5}}
 {id=Delay_1, children [Random_1], properties {delayMax=1, delayMin=0.5}}
 {id=Looping_0, children [Delay_0]}
 {id=Looping_1, children [Delay_1]}
 {id=Mixer_0, children [Modulator_3, Modulator_2, Attenuation_0], properties {inputVolume [1,1,1]}}
 {id=Mixer_1, children [SoundClass_0, Attenuation_2], properties {inputVolume [1,2]}}
 {id=Mixer_2, children [Modulator_0,Modulator_1], properties {inputVolume [1,1]}}
 {id=Modulator_0, children [Looping_0], properties {volumeMax=1.2, volumeMin=0.8}}
 {id=Modulator_1, children [Looping_1], properties {pitchMax=1.15, pitchMin=1.1, volumeMax=0.7, volumeMin=0.5}}
 {id=Modulator_2, children [Attenuation_1], properties {pitchMax=1.1, pitchMin=0.9, volumeMax=1, volumeMin=1}}
 {id=Modulator_3, children [Attenuation_3], properties {pitchMax=1.1, pitchMin=0.9, volumeMax=1, volumeMin=1}}
 {id=Random_0, children [WavePlayer_1,WavePlayer_3,WavePlayer_4], properties {weights [1,1,1]}}
 {id=Random_1, children [WavePlayer_5,WavePlayer_6,WavePlayer_7], properties {weights [1,1,1]}}
 {id=Random_2, children [WavePlayer_10, WavePlayer_11], properties {weights [1,1]}}
 {id=SoundClass_0, children [Mixer_0], properties {soundClassOverride=/Game/Sounds/FortSoundClassesAndModes/Fort_Events_Priority_ExtHigh}}
 {id=SoundClass_1, children [WavePlayer_0], properties {soundClassOverride=&Fort_Events_Priority_ExtHigh}}
 {id=WavePlayer_0, properties {soundWaveAssetPtr=&/Cane_PassBy_01}}
 {id=WavePlayer_1, properties {soundWaveAssetPtr=&/Cane_Explo_03}}
 {id=WavePlayer_10, properties {bLooping=true, soundWaveAssetPtr=&/Cane_Travel_Distant_02}}
 {id=WavePlayer_11, properties {bLooping=true, soundWaveAssetPtr=&/Cane_Travel_Distant_01}}
 {id=WavePlayer_2, properties {bLooping=true, soundWaveAssetPtr=&/Cane_Travel_Close_01}}
 {id=WavePlayer_3, properties {soundWaveAssetPtr=&/Cane_Explo_02}}
 {id=WavePlayer_4, properties {soundWaveAssetPtr=&/Cane_Explo_01}}
 {id=WavePlayer_5, properties {soundWaveAssetPtr=&Cane_Explo_03}}
 {id=WavePlayer_6, properties {soundWaveAssetPtr=&Cane_Explo_02}}
 {id=WavePlayer_7, properties {soundWaveAssetPtr=&Cane_Explo_01}}
]
properties {
 maxDistance=17500
 soundClassObject=&Fort_Events
 volumeMultiplier=1
 attenuationOverrides {bAttenuateWithLPF=true, falloffDistance=5000}
 concurrencyOverrides {maxCount=1}
}
