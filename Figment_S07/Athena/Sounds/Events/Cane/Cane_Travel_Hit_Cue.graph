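# Sound cue for the travel hit of the Cane event, playing an explosion sound in the Fort_Events sound class.
# A Random node picks one of three equally weighted Cane explosion waves and feeds a Modulator that varies its volume, with a boosted cue-level volume.
# Overridden attenuation gives it long-range falloff, a wide stereo spread and a low-pass filter that muffles it with distance.
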SoundCue /Figment_S07/Athena/Sounds/Events/Cane/Cane_Travel_Hit_Cue
root=Modulator_0
nodes [
 {id=Modulator_0, children [Random_0], properties {volumeMin=0.5}}
 {id=Random_0, children [WavePlayer_0,WavePlayer_3,WavePlayer_4], properties {weights [1,1,1]}}
 {id=WavePlayer_0, properties {soundWaveAssetPtr=&/Cane_Explo_03}}
 {id=WavePlayer_3, properties {soundWaveAssetPtr=&/Cane_Explo_02}}
 {id=WavePlayer_4, properties {soundWaveAssetPtr=&/Cane_Explo_01}}
]
properties {
 bOverrideAttenuation=true
 maxDistance=10000
 soundClassObject=&Fort_Events
 volumeMultiplier=1.5
 attenuationOverrides {
  bAttenuateWithLPF=true
  falloffDistance=10000
  lPFFrequencyAtMax=2000
  lPFRadiusMax=7000
  lPFRadiusMin=1000
  stereoSpread=5000
  attenuationShapeExtents {x=0}
 }
 concurrencyOverrides {maxCount=2, resolutionRule=StopQuietest}
}
